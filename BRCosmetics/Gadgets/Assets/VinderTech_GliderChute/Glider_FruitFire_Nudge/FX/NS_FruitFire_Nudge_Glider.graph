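# Niagara effect for the FruitFire Nudge glider cosmetic, adding fan and hair-dryer airflow visuals at the glider's sockets.
# A fan mesh, sprite speed lines and ribbon trails follow socket-reader positions, with lines spawned on rings around the sockets.
# Timelines, smoothed system values and the ForwardDot user parameter drive trail bend, length and opacity, with front-end overrides and time-of-day dimming.
# Several emitters are disabled, including earlier hair-dryer trail variants and MeggieTrail.
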NiagaraSystem /BRCosmetics/Gadgets/Assets/VinderTech_GliderChute/Glider_FruitFire_Nudge/FX/NS_FruitFire_Nudge_Glider
dependencies {
 usedDataInterfaces [
  &NiagaraDataInterfaceCurve
  &NiagaraDataInterfaceArrayFloat3
  &NiagaraDataInterfaceVector2DCurve
  &NiagaraDataInterfaceColorCurve
  &NiagaraDataInterfaceArrayColor
  &NiagaraDataInterfaceArrayFloat
  &NiagaraDataInterfaceArrayInt32
  &NiagaraDataInterfaceSocketReader
 ]
 usedDynamicInputs [
  &RemapRange
  &NormalizeVector
  &TransformVector
  &FNTimeOfDayColor
 ]
 usedModules [
  &self:SystemSpawnScript.SystemScriptSource.SystemScriptGraph.NiagaraNodeAssignment_1.SetVariables_#1
  &SystemState
  &Timeline
  &FN_Integrator
  &InterpolateOverTime
  &self:SystemSpawnScript.SystemScriptSource.SystemScriptGraph.NiagaraNodeAssignment_0.SetVariables_#2
  &self:SystemSpawnScript.SystemScriptSource.SystemScriptGraph.NiagaraNodeAssignment_2.SetVariables_#3
 ]
}
emitters {
 Fan {
  emitterUpdateScript [
   &EmitterState(
    Inactive Response=::NewEnumerator1
    Life Cycle Mode=::NewEnumerator1
    Loop Behavior=::NewEnumerator1
    Loop Duration Mode=::NewEnumerator1
    UseLoopDelay=false
   )
   &SpawnBurst_Instantaneous(Use Spawn Probability=false, UseLoopCountLimit=false)
  ]
  particleSpawnScript [
   &InitializeParticle(
    Color=<rgba 0.5 0.5 0.5 1>
    Mesh Scale Mode=::NewEnumerator3
    Mesh Uniform Scale=0.09
    Sprite Size Mode=::NewEnumerator3
    Uniform Sprite Size=80
    UsePositionOffset=false
    Write Material Random=true
   )
   &SocketLocation(
    Orientation Sampling=::NewEnumerator1
    SamplePosOffset=true
    Sampled Position Offset=<0 0 14>
    Socket Index=4
    Socket Reader=$System.Socket Reader
    Socket Sampling Mode=::NewEnumerator3
    VelocitySpeedLimit=false
   )
   &InitialMeshOrientation(Mesh Orientation Mode=::NewEnumerator2, Orientation Vector=<0 1 0>, Use Rotation Vector=false) {enabled=false}
  ]
  particleUpdateScript [
   &SolveForcesAndVelocity(Clamp Velocity=false, Limit Acceleration=false)
   &ParticleState(Kill Particles When Lifetime Has Elapsed=false)
   &SocketLocation(
    SamplePosOffset=true
    Sampled Position Offset=<0 0 14>
    Socket Index=4
    Socket Reader=$System.Socket Reader
    Socket Sampling Mode=::NewEnumerator3
    VelocitySpeedLimit=false
   ) {module=SocketLocation001}
   &DynamicMaterialParameters(
    Index 0 Param 1=&Lerp_Float(A=0.2, B=0.5, Alpha=$System.Timeline.Time)
    Index 0 Param 2=0.98
    Module.Param0WriteEnabled=true
    Module.Param1WriteEnabled=true
    Module.Param2WriteEnabled=false
    Module.Param3WriteEnabled=true
   )
   &ScaleColor(
    Scale Alpha=&FloatFromCurve(FloatCurve=DataInterface() {propertyValues {Curve {keys [<at 0 0>,<at 0.2124 1>]}}})
    ScaleA=true
    ScaleRGB=true
   ) {module=ScaleColor002}
   &ScaleMeshSize(Scale Factor=&Multiply_VectorByFloat(
     Vector=<1 1 1>
     Float=&FloatFromCurve(FloatCurve=DataInterface() {propertyValues {Curve {keys [<at 0 0>,<at 1 1>]}}}, CurveIndex=$Emitter.Age)
    )) {enabled=false}
  ]
  rendererClasses [&NiagaraMeshRendererProperties]
  renderers [{rendererClass=&NiagaraMeshRendererProperties, rendererIndex=0}]
 }
 FanLines {
  emitterUpdateScript [
   &EmitterState()
   &SpawnRate(SpawnRate=40, Use Spawn Probability=false)
  ]
  particleSpawnScript [
   &InitializeParticle(
    Color=<rgba 2 2 2 1>
    Lifetime=0.2
    Sprite Rotation Mode=::NewEnumerator2
    Sprite Size Mode=::NewEnumerator3
    Uniform Sprite Size=5
    UsePositionOffset=false
    Write Material Random=true
   )
   &SocketLocation(
    Position Sampling=::NewEnumerator1
    SamplePosOffset=true
    Sampled Position Offset=<0 0 15>
    Socket Index=4
    Socket Reader=$System.Socket Reader
    Socket Sampling Mode=::NewEnumerator3
    VelocitySpeedLimit=false
   )
   &AddVelocity(
    Cone Angle=20
    Cone Axis=&Add_Vector(
     A=<0 0 0>
     B=&RotateVectorByQuaternion(Vector To Rotate=<0 1 0>, Quaternion=$Output.SocketLocation.SampledRotation)
    )
    Inner Cone Angle=20
    Module.UseConeFalloff=false
    Module.UseVelDistribution=false
    Override Seed=false
    Velocity Mode=::NewEnumerator2
    Velocity Speed=300
   ) {enabled=false}
   &ShapeLocation(
    Disc Coverage=0.8333
    MaskSpawnGroup=false
    Override Seed=false
    Ring Radius=15
    Rotation Angle Type:ENiagara_AngleInput=::NewEnumerator0
    Rotation Axis=&RotateVectorByQuaternion(Vector To Rotate=<0 0 1>, Quaternion=$Output.SocketLocation.SampledRotation)
    Rotation Mode=::NewEnumerator1
    Shape Origin=$Output.SocketLocation.SampledPosition
    Shape Primitive=::NewEnumerator4
    UseOwnerScale=true
   )
   &AddVelocity(
    Additional Quaternion Rotation=false
    Rotation Mode=::NewEnumerator3
    Rotation Quaternion:Quat4f=$Output.SocketLocation.SampledRotation
    Velocity=<0 0 1>
   ) {module=AddVelocity001}
   &ApplyInitialForces() {enabled=false}
  ]
  particleUpdateScript [
   &ParticleState()
   &ScaleSpriteSize(
    Non-Uniform Curve Scale=&MakeVector2D(
     X=0.2
     Y=&FN_FrontEndFloatModifier(
      Gameplay=&Lerp_Float(
       A=&RandomRangeFloat(Minimum=8, Maximum=6)
       B=&RandomRangeFloat(Minimum=9, Maximum=12)
       Alpha=&ClampFloat(Float=$User.ForwardDot)
      )
      Front End=&RandomRangeFloat(Minimum=11, Maximum=13)
     )
    )
    Scale Sprite Size Mode=::NewEnumerator3
   )
   &Drag(Use Linear Drag=true, Use Rotational Drag=true) {enabled=false}
   &SolveForcesAndVelocity(Clamp Velocity=false, Limit Acceleration=false)
   &ScaleSpriteSizeBySpeed(Max Scale Factor=<1 10>, Min Scale Factor=<1 5>, Sample Scale Factor By Curve=true) {enabled=false}
   &DynamicMaterialParameters(Module.Index0Float0Write=true, Module.Index0Float1Write=true, Module.Index0Float2Write=true, Module.Index0Float3Write=true)
   &ScaleColor(
    Scale Alpha=&FloatFromCurve(
     FloatCurve=DataInterface() {propertyValues {Curve {keys [
         <at 0 0 RCIM_Cubic>
         <at 0.109 0.9794 RCIM_Cubic>
         {arriveTangent=-1.6264, interpMode=RCIM_Cubic, leaveTangent=-1.6264, tangentMode=RCTM_User, time=0.7653, value=0}
        ]}}}
     Scale Curve=&Lerp_Float(A=4, B=10, Alpha=&ClampFloat(Float=$User.ForwardDot))
    )
    ScaleA=true
    ScaleRGB=true
   )
   &ScaleColor(Scale Alpha=$System.TOD, ScaleA=true, ScaleRGB=true) {module=ScaleColor001}
  ]
  rendererClasses [&NiagaraSpriteRendererProperties]
  renderers [{rendererClass=&NiagaraSpriteRendererProperties}]
 }
 FanTrail {
  emitterUpdateScript [
   &EmitterState()
   &SpawnBurst_Instantaneous(Spawn Count=50, Use Spawn Probability=false, UseLoopCountLimit=true)
  ]
  particleSpawnScript [
   &PartitionParticles(Number Of Partitions=1, Write Partition to Ribbon ID=::NewEnumerator0)
   &InitializeParticle(
    Color=<rgba 3 3 3 10>
    Lifetime=0.2
    Position Offset=<0 0 25>
    Ribbon Width=22
    Ribbon Width Mode=::NewEnumerator1
    Sprite Size Mode=::NewEnumerator3
    Uniform Sprite Size=5
    UsePositionOffset=false
    Write Material Random=true
   )
   /SetVariables(Particles.SpawnGroup=$Particles.PartitionParticles.Partition) {isSetParameters=true}
   /SetVariables(
    Particles.RibbonID=/Niagara/DynamicInputs/Meshes/MakeRibbonIDFromSpawnGroup()
    Particles.RibbonLinkOrder:NiagaraFloat=&ReturnNormalizedExecIndex()
   ) {isSetParameters=true, module=SetVariables_#4}
  ]
  particleUpdateScript [
   &ParticleState(Kill Particles When Lifetime Has Elapsed=false, Let Infinitely Lived Particles Die When Emitter Deactivates=true)
   &SocketLocation(
    SamplePosOffset=true
    Sampled Orientation Coordinate Space:ENiagaraCoordinateSpace=::Local
    Sampled Position Offset=<0 0 15>
    Socket Index=4
    Socket Reader=$System.Socket Reader
    Socket Sampling Mode=::NewEnumerator3
    VelocitySpeedLimit=false
   )
   /SetVariables(Particles.Position=&LerpPosition(
     A=$Particles.Position
     B=&AddVectorToPosition(
      Position=$Particles.Position
      Vector=&MakeVector(
       X=&FloatFromCurve(
        FloatCurve=DataInterface() {propertyValues {Curve {keys [<at 0 0 RCIM_Cubic>, {arriveTangent=2.1213, interpMode=RCIM_Cubic, leaveTangent=2.1213, tangentMode=RCTM_User, time=1, value=1}]}}}
        CurveIndex=&ReturnNormalizedExecIndex()
        Scale Curve=&Multiply_Float(
         A=&MakeFloatFromVector(Vector=$System.InterpolateOverTime.InterpolateOverTime_V3_.Moving Average, Channel=::NewEnumerator0)
         B=&Lerp_Float(A=0, B=-100, Alpha=$System.InterpolateOverTime001.InterpolateOverTime_Float.Moving Average)
        )
       )
       Y=&Multiply_Float(
        A=&FloatFromCurve(
         FloatCurve=DataInterface() {propertyValues {Curve {keys [{tangentMode=RCTM_User, time=0, value=0}, {tangentMode=RCTM_User, time=1, value=1}]}}}
         CurveIndex=&ReturnNormalizedExecIndex()
         Scale Curve=&FN_FrontEndFloatModifier(
          Gameplay=&Lerp_Float(A=50, B=200, Alpha=$System.Timeline.NormalizedTime)
          Front End=&Lerp_Float(A=50, B=200, Alpha=$System.Timeline002.NormalizedTime)
         )
        )
        B=&FloatFromCurve(FloatCurve=DataInterface() {propertyValues {Curve {defaultValue=1, keys [{tangentMode=RCTM_User, time=0, value=0}, {tangentMode=RCTM_User, time=3, value=1}]}}})
       )
       Z=0
      )
     )
     Alpha=&ReturnNormalizedExecIndex()
    )) {isSetParameters=true, module=SetVariables_#5}
   &ScaleColor(
    Curve Index=$Particles.RibbonLinkOrder
    Linear Color Curve=DataInterface() {propertyValues {
      AlphaCurve {keys [<at 0.0886 1>,<at 0.9269 1>,<at 1 0>]}
      BlueCurve {keys [<at 0.9957 1>]}
      GreenCurve {keys [<at 0.9957 1>]}
      RedCurve {keys [<at 0.9957 1>]}
     }}
    Scale Mode=::NewEnumerator2
   )
   &ScaleColor(
    Scale Alpha=&FloatFromCurve(
     FloatCurve=DataInterface() {propertyValues {Curve {keys [{tangentMode=RCTM_User, time=0, value=1}, {tangentMode=RCTM_User, time=1, value=0}]}}}
     CurveIndex=$Particles.Initial.RibbonLinkOrder
     Scale Curve=&FN_FrontEndFloatModifier(
      Gameplay=&Lerp_Float(A=0.025, B=0.05, Alpha=$System.Timeline.NormalizedTime)
      Front End=&Lerp_Float(A=0.025, B=0.025, Alpha=$System.Timeline002.NormalizedTime)
     )
    )
    ScaleA=true
    ScaleRGB=true
   ) {module=ScaleColor001}
   /SetVariables(
    Particles.LineTrailColor:LinearColor=&MakeLinearColorFromVectorAndFloat(
     "Vector (RGB)"=<0.05 0.05 0.05>
     "Float (Alpha)"=&FloatFromCurve(
      FloatCurve=DataInterface() {propertyValues {Curve {keys [{tangentMode=RCTM_User, time=0, value=0}, <at 0.0841 1.0029>, {tangentMode=RCTM_User, time=0.7064, value=0}]}}}
      CurveIndex=$Particles.Initial.RibbonLinkOrder
      Scale Curve=&FN_FrontEndFloatModifier(
       Gameplay=&Multiply_Float(
        A=&Lerp_Float(A=0.2, B=0.4, Alpha=$System.Timeline.NormalizedTime)
        B=&FNTimeOfDayScalar(Night Scale=0.5, Morning Scale=0.75, Evening Scale=0.75)
       )
       Front End=&Lerp_Float(A=0.2, B=0.4, Alpha=$System.Timeline002.NormalizedTime)
      )
     )
    )
    Particles.LineTrailDynamic:Vector4f=&MakeVector4(
     X=&FloatFromCurve(FloatCurve=DataInterface() {propertyValues {Curve {keys [<at 0 1>,<at 1 0>]}}}, CurveIndex=$Particles.RibbonLinkOrder, Scale Curve=4)
     Y=0
     Z=0
     W=0
    )
    Particles.LineTrailWidth:NiagaraFloat=&FloatFromCurve(
     FloatCurve=DataInterface() {propertyValues {Curve {keys [{arriveTangent=-0.0469, interpMode=RCIM_Cubic, leaveTangent=-0.0469, tangentMode=RCTM_User, time=0, value=1}]}}}
     CurveIndex=$Particles.Initial.RibbonLinkOrder
     Scale Curve=30
    )
   ) {isSetParameters=true}
   &ScaleRibbonWidth(Ribbon Width Scale=&FloatFromCurve(
     FloatCurve=DataInterface() {propertyValues {Curve {keys [{tangentMode=RCTM_User, time=0, value=1}, {tangentMode=RCTM_User, time=1, value=2}]}}}
     CurveIndex=$Particles.Initial.RibbonLinkOrder
     Scale Curve=2.5
    ))
   &ScaleColor(
    Scale Alpha=&MakeCustomFloatFromBool(
     Bool=&SetBoolByFloatComparison(A=$User.ForwardDot, B=0)
     True Float=1
     False Float=&NormalizeFloat(
      Value=&OneMinusFloat(Float=&Multiply_Float(A=$System.InterpolateOverTime001.InterpolateOverTime_Float.Moving Average, B=2))
      Min=0
      Max=1
     )
    )
    ScaleA=true
    ScaleRGB=true
   ) {enabled=false, module=ScaleColor002}
   &DynamicMaterialParameters(
    Index 0 Param 1=&FloatFromCurve(
     FloatCurve=DataInterface() {propertyValues {Curve {defaultValue=1, keys [{tangentMode=RCTM_User, time=0, value=0}, {tangentMode=RCTM_User, time=1, value=1}]}}}
     Scale Curve=10
    )
    Index 0 Param 2=1
    Index 0 Param 3=1
    Index 0 Param 4=1
    Module.Index0Float0Write=true
    Module.Index0Float1Write=true
    Module.Index0Float2Write=true
    Module.Index0Float3Write=true
   ) {enabled=false}
   &ScaleColor(
    Scale Alpha=&FNTimeOfDayScalar(Night Scale=0.3, Morning Scale=0.5, Day Scale=0.75, Evening Scale=0.5)
    Scale RGB=&Multiply_VectorByFloat(Vector=<1 1 1>, Float=&FNTimeOfDayScalar(Night Scale=0.5, Morning Scale=1, Evening Scale=1))
    ScaleA=true
    ScaleRGB=true
   ) {module=ScaleColor003}
  ]
  rendererClasses [&NiagaraSpriteRendererProperties, &NiagaraRibbonRendererProperties]
  renderers [
   {rendererClass=&NiagaraSpriteRendererProperties}
   {rendererClass=&NiagaraRibbonRendererProperties, rendererIndex=1}
   {rendererClass=&NiagaraRibbonRendererProperties, rendererIndex=2}
  ]
 }
 HairDryerLines {
  emitterUpdateScript [
   &EmitterState()
   &SpawnRate(SpawnRate=80, Use Spawn Probability=false)
  ]
  particleSpawnScript [
   /SetVariables(Particles.SocketIndex:NiagaraInt32=&RandomRangeInt(Minimum=6, Maximum=7)) {isSetParameters=true}
   &InitializeParticle(
    Color=&MaskLinearColor(
     Mask Source=1
     Mask Parameter=$Particles.SocketIndex
     Mask=6
     Masked Linear Color=&SelectLinearColorFromArray(
      Array Sampling Mode=::NewEnumerator0
      Direct Array Index=?
      Color Selection Array=DataInterface() {propertyValues {ColorData [<rgba 0.292 1.0111 2 1>, <rgba 1 1 1 1>]}}
      Random Seed=0
     )
     Passthrough Linear Color=&SelectLinearColorFromArray(
      Array Sampling Mode=::NewEnumerator0
      Direct Array Index=?
      Color Selection Array=DataInterface() {propertyValues {ColorData [<rgba 1.5596 0.519 2 1>, <rgba 1 1 1 1>]}}
      Random Seed=0
     )
    )
    Lifetime=0.2
    Sprite Rotation Mode=::NewEnumerator2
    Sprite Size Mode=::NewEnumerator3
    Uniform Sprite Size=5
    UsePositionOffset=false
    Write Material Random=true
   )
   &SocketLocation(
    Position Sampling=::NewEnumerator1
    SamplePosOffset=false
    Socket Index=$Particles.SocketIndex
    Socket Reader=$System.Socket Reader
    Socket Sampling Mode=::NewEnumerator3
    VelocitySpeedLimit=false
   )
   &AddVelocity(
    Cone Angle=20
    Cone Axis=&Add_Vector(
     A=<0 0 0>
     B=&RotateVectorByQuaternion(Vector To Rotate=<0 1 0>, Quaternion=$Output.SocketLocation.SampledRotation)
    )
    Inner Cone Angle=20
    Module.UseConeFalloff=false
    Module.UseVelDistribution=false
    Override Seed=false
    Velocity Mode=::NewEnumerator2
    Velocity Speed=300
   ) {enabled=false}
   &ShapeLocation(
    Disc Coverage=1
    MaskSpawnGroup=false
    Override Seed=false
    Ring Radius=&SelectFloatFromArray(
     Direct Array Index=/Niagara/DynamicInputs/Subtract/Subtract_Integer(A:NiagaraInt32=$Particles.SocketIndex, B:NiagaraInt32=6)
     Float Selection Array=DataInterface() {propertyValues {FloatData [1.5,2.5]}}
    )
    Rotation Angle Type:ENiagara_AngleInput=::NewEnumerator0
    Rotation Axis=&RotateVectorByQuaternion(Vector To Rotate=<0 1 0>, Quaternion=$Output.SocketLocation.SampledRotation)
    Rotation Mode=::NewEnumerator1
    Shape Origin=$Output.SocketLocation.SampledPosition
    Shape Primitive=::NewEnumerator4
    UseOwnerScale=true
   )
   &AddVelocity(
    Additional Quaternion Rotation=false
    Rotation Mode=::NewEnumerator3
    Rotation Quaternion:Quat4f=$Output.SocketLocation.SampledRotation
    Velocity=<0 1 0>
   ) {module=AddVelocity001}
   &ApplyInitialForces()
   &AlignSpriteToMeshOrientation(
    Mesh Orientation Relative Sprite Alignment Vector=<1 1 1>
    Mesh Orientation Relative Sprite Facing Vector:Vector3f=<0 0 1>
    SetAlignment=true
    SetFacing=true
   ) {enabled=false}
  ]
  particleUpdateScript [
   &ParticleState()
   &ScaleSpriteSize(
    Non-Uniform Curve Scale=&MakeVector2D(
     X=0.2
     Y=&FN_FrontEndFloatModifier(
      Gameplay=&Lerp_Float(
       A=&RandomRangeFloat(Minimum=7, Maximum=9)
       B=&RandomRangeFloat(Minimum=11, Maximum=13)
       Alpha=&ClampFloat(Float=$User.ForwardDot)
      )
      Front End=&RandomRangeFloat(Minimum=11, Maximum=13)
     )
    )
    Scale Sprite Size Mode=::NewEnumerator3
   )
   &SocketLocation(
    Position Sampling=::NewEnumerator1
    SamplePosOffset=false
    Socket Index=$Particles.SocketIndex
    Socket Reader=$System.Socket Reader
    Socket Sampling Mode=::NewEnumerator3
    VelocitySpeedLimit=false
   ) {module=SocketLocation001}
   &ShapeLocation(
    Disc Coverage=1
    MaskSpawnGroup=false
    Override Seed=false
    Ring Radius=&SelectFloatFromArray(
     Direct Array Index=&Subtract_Integer(A:NiagaraInt32=$Particles.SocketIndex, B:NiagaraInt32=6)
     Float Selection Array=DataInterface() {propertyValues {FloatData [2,2.75]}}
    )
    Rotation Angle Type:ENiagara_AngleInput=::NewEnumerator0
    Rotation Axis=&RotateVectorByQuaternion(Vector To Rotate=<0 1 0>, Quaternion=$Output.SocketLocation001.SampledRotation)
    Rotation Mode=::NewEnumerator1
    Shape Origin=$Output.SocketLocation001.SampledPosition
    Shape Primitive=::NewEnumerator4
    UseOwnerScale=true
   ) {module=ShapeLocation001}
   &Drag(Use Linear Drag=true, Use Rotational Drag=true) {enabled=false}
   &SolveForcesAndVelocity(Clamp Velocity=false, Limit Acceleration=false) {enabled=false}
   &ScaleSpriteSizeBySpeed(Max Scale Factor=<1 10>, Min Scale Factor=<1 5>, Sample Scale Factor By Curve=true) {enabled=false}
   &DynamicMaterialParameters(Module.Index0Float0Write=true, Module.Index0Float1Write=true, Module.Index0Float2Write=true, Module.Index0Float3Write=true)
   &ScaleColor(
    Scale Alpha=&FloatFromCurve(
     FloatCurve=DataInterface() {propertyValues {Curve {keys [<at 0.03 1 RCIM_Cubic>, {arriveTangent=-1.6264, interpMode=RCIM_Cubic, leaveTangent=-1.6264, tangentMode=RCTM_User, time=0.7653, value=0}]}}}
     Scale Curve=&Lerp_Float(A=5, B=10, Alpha=&ClampFloat(Float=$User.ForwardDot))
    )
    ScaleA=true
    ScaleRGB=true
   )
   &ScaleColor(Scale Alpha=$System.TOD, ScaleA=true, ScaleRGB=true) {module=ScaleColor001}
  ]
  rendererClasses [&NiagaraSpriteRendererProperties]
  renderers [{rendererClass=&NiagaraSpriteRendererProperties}]
 }
 HairDryerTrail {
  bEnabled=false
  emitterUpdateScript [
   &EmitterState()
   &SpawnBurst_Instantaneous(Spawn Count=30, Use Spawn Probability=false, UseLoopCountLimit=true)
  ]
  particleSpawnScript [
   &InitializeParticle(
    Lifetime=0.2
    Position Offset=<0 0 25>
    Ribbon Width=22
    Ribbon Width Mode=::NewEnumerator1
    Sprite Size Mode=::NewEnumerator3
    Uniform Sprite Size=5
    UsePositionOffset=false
    Write Material Random=true
   )
   &FN_SetSpawnGroup(Number of Spawn Groups=2)
   /SetVariables(
    Particles.RibbonID=&MakeRibbonIDFromSpawnGroup()
    Particles.RibbonLinkOrder:NiagaraFloat=&ReturnNormalizedExecIndex()
   ) {isSetParameters=true, module=SetVariables_#4}
  ]
  particleUpdateScript [
   &ParticleState(Kill Particles When Lifetime Has Elapsed=false, Let Infinitely Lived Particles Die When Emitter Deactivates=true)
   &SocketLocation(
    SamplePosOffset=true
    Sampled Position Offset=&MakeVector(
     X=&Lerp_Float(
      A=0
      B=&FloatFromCurve(
       FloatCurve=DataInterface() {propertyValues {Curve {keys [
           {arriveTangent=0.085, interpMode=RCIM_Cubic, leaveTangent=0.085, tangentMode=RCTM_User, time=0.2559, value=0}
           {arriveTangent=0.9065, interpMode=RCIM_Cubic, leaveTangent=0.9065, tangentMode=RCTM_User, time=1, value=0.3081}
          ]}}}
       CurveIndex=&ReturnNormalizedExecIndex()
       Scale Curve=&Multiply_Float(A=-200, B=$System.Timeline001.Time)
      )
      Alpha=$System.InterpolateOverTime001.InterpolateOverTime_Float.Moving Average
     )
     Y=&FloatFromCurve(
      FloatCurve=DataInterface() {propertyValues {Curve {keys [{tangentMode=RCTM_User, time=0, value=0}, {tangentMode=RCTM_User, time=1, value=1}]}}}
      CurveIndex=&ReturnNormalizedExecIndex()
      Scale Curve=&FN_FrontEndFloatModifier(
       Gameplay=&Lerp_Float(A=75, B=200, Alpha=$System.Timeline.Time)
       Front End=&Lerp_Float(A=75, B=200, Alpha=$System.Timeline002.Time)
      )
     )
     Z=0
    )
    Socket Index=&SelectIntFromArray(Direct Array Index=$Particles.SpawnGroup, Int Selection Array=DataInterface() {propertyValues {IntData [6,7]}})
    Socket Reader=$System.Socket Reader
    Socket Sampling Mode=::NewEnumerator3
    VelocitySpeedLimit=false
   )
   /SetVariables(Particles.Position=&LerpPosition(
     A=$Particles.Position
     B=&AddVectorToPosition(
      Position=$Particles.Position
      Vector=&MakeVector(
       X=&FloatFromCurve(
        FloatCurve=DataInterface() {propertyValues {Curve {keys [
            {arriveTangent=-2.395, interpMode=RCIM_Cubic, leaveTangent=-2.395, tangentMode=RCTM_User, time=0, value=0}
            {arriveTangent=0.8, interpMode=RCIM_Cubic, leaveTangent=0.8, tangentMode=RCTM_User, time=1, value=0.3081}
           ]}}}
        CurveIndex=&ReturnNormalizedExecIndex()
        Scale Curve=&Multiply_Float(
         A=&MakeFloatFromVector(Vector=$System.InterpolateOverTime.InterpolateOverTime_V3_.Moving Average, Channel=::NewEnumerator0)
         B=&Lerp_Float(
          A=0
          B=&Lerp_Float(A=25, B=100, Alpha=$System.Timeline.Time)
          Alpha=$System.InterpolateOverTime001.InterpolateOverTime_Float.Moving Average
         )
        )
       )
       Y=&FloatFromCurve(
        FloatCurve=DataInterface() {propertyValues {Curve {keys [{tangentMode=RCTM_User, time=0, value=0}, {tangentMode=RCTM_User, time=1, value=1}]}}}
        CurveIndex=&ReturnNormalizedExecIndex()
        Scale Curve=&FN_FrontEndFloatModifier(
         Gameplay=&Lerp_Float(A=50, B=200, Alpha=$System.Timeline.NormalizedTime)
         Front End=&Lerp_Float(A=0, B=200, Alpha=$System.Timeline002.NormalizedTime)
        )
       )
       Z=0
      )
     )
     Alpha=&ReturnNormalizedExecIndex()
    )) {enabled=false, isSetParameters=true, module=SetVariables_#5}
   &ScaleColor(
    Curve Index=$Particles.RibbonLinkOrder
    Linear Color Curve=DataInterface() {propertyValues {
      AlphaCurve {keys [<at 0.0886 1>,<at 0.9269 1>,<at 1 0>]}
      BlueCurve {keys [<at 0.9957 1>]}
      GreenCurve {keys [<at 0.9957 1>]}
      RedCurve {keys [<at 0.9957 1>]}
     }}
    Scale Mode=::NewEnumerator2
   )
   &ScaleColor(
    Scale Alpha=&FloatFromCurve(
     FloatCurve=DataInterface() {propertyValues {Curve {keys [{tangentMode=RCTM_User, time=0, value=1}, {tangentMode=RCTM_User, time=1, value=0}]}}}
     CurveIndex=$Particles.Initial.RibbonLinkOrder
     Scale Curve=&FN_FrontEndFloatModifier(
      Gameplay=&Lerp_Float(A=0.5, B=0.75, Alpha=$System.Timeline.NormalizedTime)
      Front End=&Lerp_Float(A=0, B=0.4, Alpha=$System.Timeline002.NormalizedTime)
     )
    )
    ScaleA=true
    ScaleRGB=true
   ) {module=ScaleColor001}
   &ScaleRibbonWidth(Ribbon Width Scale=&FloatFromCurve(
     FloatCurve=DataInterface() {propertyValues {Curve {keys [{tangentMode=RCTM_User, time=0, value=0.75}, <at 0.484 2.1885>, {tangentMode=RCTM_User, time=1, value=2}]}}}
     CurveIndex=$Particles.Initial.RibbonLinkOrder
    ))
   &DynamicMaterialParameters(
    Index 0 Param 1=&MaskFloat(Mask Source=0, Mask Parameter=?, Mask=0, Masked Float=0, Passthrough Float=1)
    Index 0 Param 2=1
    Index 0 Param 3=1
    Index 0 Param 4=1
    Module.Index0Float0Write=true
    Module.Index0Float1Write=true
    Module.Index0Float2Write=true
    Module.Index0Float3Write=true
   )
   &ScaleColor(
    Scale Alpha=&FNTimeOfDayScalar(Night Scale=0.5, Morning Scale=0.75, Evening Scale=0.75)
    ScaleA=true
    ScaleRGB=true
   ) {module=ScaleColor002}
  ]
  rendererClasses [&NiagaraSpriteRendererProperties, &NiagaraRibbonRendererProperties]
  renderers [
   {rendererClass=&NiagaraSpriteRendererProperties}
   {rendererClass=&NiagaraRibbonRendererProperties, rendererIndex=1}
   {rendererClass=&NiagaraRibbonRendererProperties, rendererIndex=2}
  ]
 }
 HairDryerTrail001 {
  bEnabled=false
  emitterUpdateScript [
   &EmitterState()
   &SpawnBurst_Instantaneous(Spawn Count=30, Use Spawn Probability=false, UseLoopCountLimit=true)
  ]
  particleSpawnScript [
   &InitializeParticle(
    Lifetime=0.2
    Position Offset=<0 0 25>
    Ribbon Width=22
    Ribbon Width Mode=::NewEnumerator1
    Sprite Size Mode=::NewEnumerator3
    Uniform Sprite Size=5
    UsePositionOffset=false
    Write Material Random=true
   )
   &FN_SetSpawnGroup(Number of Spawn Groups=2)
   /SetVariables(
    Particles.RibbonID=&MakeRibbonIDFromSpawnGroup()
    Particles.RibbonLinkOrder:NiagaraFloat=&ReturnNormalizedExecIndex()
   ) {isSetParameters=true, module=SetVariables_#4}
  ]
  particleUpdateScript [
   &ParticleState(Kill Particles When Lifetime Has Elapsed=false, Let Infinitely Lived Particles Die When Emitter Deactivates=true)
   &SocketLocation(
    SamplePosOffset=true
    Sampled Position Offset=&MakeVector(
     X=&Lerp_Float(
      A=0
      B=&FloatFromCurve(
       FloatCurve=DataInterface() {propertyValues {Curve {keys [
           {arriveTangent=0.085, interpMode=RCIM_Cubic, leaveTangent=0.085, tangentMode=RCTM_User, time=0.2559, value=0}
           {arriveTangent=0.9065, interpMode=RCIM_Cubic, leaveTangent=0.9065, tangentMode=RCTM_User, time=1, value=0.3081}
          ]}}}
       CurveIndex=&ReturnNormalizedExecIndex()
       Scale Curve=&Multiply_Float(A=-200, B=$System.Timeline001.Time)
      )
      Alpha=$System.InterpolateOverTime001.InterpolateOverTime_Float.Moving Average
     )
     Y=&FloatFromCurve(
      FloatCurve=DataInterface() {propertyValues {Curve {keys [{tangentMode=RCTM_User, time=0, value=0}, {tangentMode=RCTM_User, time=1, value=1}]}}}
      CurveIndex=&ReturnNormalizedExecIndex()
      Scale Curve=&FN_FrontEndFloatModifier(
       Gameplay=&Lerp_Float(A=75, B=200, Alpha=$System.Timeline.Time)
       Front End=&Lerp_Float(A=75, B=200, Alpha=$System.Timeline002.Time)
      )
     )
     Z=0
    )
    Socket Index=&SelectIntFromArray(Direct Array Index=$Particles.SpawnGroup, Int Selection Array=DataInterface() {propertyValues {IntData [6,7]}})
    Socket Reader=$System.Socket Reader
    Socket Sampling Mode=::NewEnumerator3
    VelocitySpeedLimit=false
   )
   /SetVariables(Particles.Position=&LerpPosition(
     A=$Particles.Position
     B=&AddVectorToPosition(
      Position=$Particles.Position
      Vector=&MakeVector(
       X=&FloatFromCurve(
        FloatCurve=DataInterface() {propertyValues {Curve {keys [
            {arriveTangent=-2.395, interpMode=RCIM_Cubic, leaveTangent=-2.395, tangentMode=RCTM_User, time=0, value=0}
            {arriveTangent=0.8, interpMode=RCIM_Cubic, leaveTangent=0.8, tangentMode=RCTM_User, time=1, value=0.3081}
           ]}}}
        CurveIndex=&ReturnNormalizedExecIndex()
        Scale Curve=&Multiply_Float(
         A=&MakeFloatFromVector(Vector=$System.InterpolateOverTime.InterpolateOverTime_V3_.Moving Average, Channel=::NewEnumerator0)
         B=&Lerp_Float(
          A=0
          B=&Lerp_Float(A=25, B=100, Alpha=$System.Timeline.Time)
          Alpha=$System.InterpolateOverTime001.InterpolateOverTime_Float.Moving Average
         )
        )
       )
       Y=&FloatFromCurve(
        FloatCurve=DataInterface() {propertyValues {Curve {keys [{tangentMode=RCTM_User, time=0, value=0}, {tangentMode=RCTM_User, time=1, value=1}]}}}
        CurveIndex=&ReturnNormalizedExecIndex()
        Scale Curve=&FN_FrontEndFloatModifier(
         Gameplay=&Lerp_Float(A=50, B=200, Alpha=$System.Timeline.NormalizedTime)
         Front End=&Lerp_Float(A=0, B=200, Alpha=$System.Timeline002.NormalizedTime)
        )
       )
       Z=0
      )
     )
     Alpha=&ReturnNormalizedExecIndex()
    )) {enabled=false, isSetParameters=true, module=SetVariables_#5}
   &ScaleColor(
    Curve Index=$Particles.RibbonLinkOrder
    Linear Color Curve=DataInterface() {propertyValues {
      AlphaCurve {keys [<at 0.0886 1>,<at 0.9269 1>,<at 1 0>]}
      BlueCurve {keys [<at 0.9957 1>]}
      GreenCurve {keys [<at 0.9957 1>]}
      RedCurve {keys [<at 0.9957 1>]}
     }}
    Scale Mode=::NewEnumerator2
   )
   &ScaleColor(
    Scale Alpha=&FloatFromCurve(
     FloatCurve=DataInterface() {propertyValues {Curve {keys [{tangentMode=RCTM_User, time=0, value=0}, <at 0.0841 1.0029>, {tangentMode=RCTM_User, time=0.7064, value=0}]}}}
     CurveIndex=$Particles.Initial.RibbonLinkOrder
     Scale Curve=&FN_FrontEndFloatModifier(
      Gameplay=&Lerp_Float(A=0.05, B=0.075, Alpha=$System.Timeline.NormalizedTime)
      Front End=&Lerp_Float(A=0, B=0.4, Alpha=$System.Timeline002.NormalizedTime)
     )
    )
    ScaleA=true
    ScaleRGB=true
   ) {module=ScaleColor001}
   &ScaleRibbonWidth(Ribbon Width Scale=&FloatFromCurve(
     FloatCurve=DataInterface() {propertyValues {Curve {keys [
         {arriveTangent=1.4042, interpMode=RCIM_Cubic, leaveTangent=1.4042, tangentMode=RCTM_User, time=0, value=0.15}
         {tangentMode=RCTM_User, time=1, value=0.8}
        ]}}}
     CurveIndex=$Particles.Initial.RibbonLinkOrder
    ))
   &DynamicMaterialParameters(
    Index 0 Param 1=&FloatFromCurve(FloatCurve=DataInterface() {propertyValues {Curve {keys [<at 0 1>,<at 1 0>]}}}, CurveIndex=$Particles.RibbonLinkOrder, Scale Curve=10)
    Module.Index0Float0Write=true
    Module.Index0Float1Write=true
    Module.Index0Float2Write=true
    Module.Index0Float3Write=true
   )
   &ScaleColor(
    Scale Alpha=&FNTimeOfDayScalar(Night Scale=0.5, Morning Scale=0.75, Evening Scale=0.75)
    ScaleA=true
    ScaleRGB=true
   ) {module=ScaleColor002}
  ]
  rendererClasses [&NiagaraSpriteRendererProperties, &NiagaraRibbonRendererProperties]
  renderers [
   {rendererClass=&NiagaraSpriteRendererProperties}
   {rendererClass=&NiagaraRibbonRendererProperties, rendererIndex=1}
   {rendererClass=&NiagaraRibbonRendererProperties, rendererIndex=2}
  ]
 }
 HairDryerTrail002 {
  emitterUpdateScript [
   &EmitterState()
   &SpawnBurst_Instantaneous(Spawn Count=30, Use Spawn Probability=false, UseLoopCountLimit=true)
  ]
  particleSpawnScript [
   &InitializeParticle(
    Lifetime=0.2
    Position Offset=<0 0 25>
    Ribbon Width=22
    Ribbon Width Mode=::NewEnumerator1
    Sprite Size Mode=::NewEnumerator3
    Uniform Sprite Size=5
    UsePositionOffset=false
    Write Material Random=true
   )
   &FN_SetSpawnGroup(Number of Spawn Groups=2)
   /SetVariables(
    Particles.RibbonID=&MakeRibbonIDFromSpawnGroup()
    Particles.RibbonLinkOrder:NiagaraFloat=&ReturnNormalizedExecIndex()
   ) {isSetParameters=true, module=SetVariables_#4}
  ]
  particleUpdateScript [
   &ParticleState(Kill Particles When Lifetime Has Elapsed=false, Let Infinitely Lived Particles Die When Emitter Deactivates=true)
   &SocketLocation(
    SamplePosOffset=true
    Sampled Position Offset=&MakeVector(
     X=&Lerp_Float(
      A=0
      B=&FloatFromCurve(
       FloatCurve=DataInterface() {propertyValues {Curve {keys [
           {arriveTangent=0.085, interpMode=RCIM_Cubic, leaveTangent=0.085, tangentMode=RCTM_User, time=0.2559, value=0}
           {arriveTangent=0.9065, interpMode=RCIM_Cubic, leaveTangent=0.9065, tangentMode=RCTM_User, time=1, value=0.3081}
          ]}}}
       CurveIndex=&ReturnNormalizedExecIndex()
       Scale Curve=&Multiply_Float(A=-200, B=$System.Timeline001.Time)
      )
      Alpha=$System.InterpolateOverTime001.InterpolateOverTime_Float.Moving Average
     )
     Y=&Multiply_Float(
      A=&FloatFromCurve(
       FloatCurve=DataInterface() {propertyValues {Curve {keys [{tangentMode=RCTM_User, time=0, value=0}, {tangentMode=RCTM_User, time=1, value=1}]}}}
       CurveIndex=&ReturnNormalizedExecIndex()
       Scale Curve=&FN_FrontEndFloatModifier(
        Gameplay=&Lerp_Float(A=75, B=200, Alpha=$System.Timeline.Time)
        Front End=&Lerp_Float(A=75, B=200, Alpha=$System.Timeline002.Time)
       )
      )
      B=&FloatFromCurve(FloatCurve=DataInterface() {propertyValues {Curve {defaultValue=1, keys [{tangentMode=RCTM_User, time=0, value=0}, {tangentMode=RCTM_User, time=3, value=1}]}}})
     )
     Z=0
    )
    Socket Index=&SelectIntFromArray(Direct Array Index=$Particles.SpawnGroup, Int Selection Array=DataInterface() {propertyValues {IntData [6,7]}})
    Socket Reader=$System.Socket Reader
    Socket Sampling Mode=::NewEnumerator3
    VelocitySpeedLimit=false
   )
   /SetVariables(Particles.Position=&LerpPosition(
     A=$Particles.Position
     B=&AddVectorToPosition(
      Position=$Particles.Position
      Vector=&MakeVector(
       X=&FloatFromCurve(
        FloatCurve=DataInterface() {propertyValues {Curve {keys [
            {arriveTangent=-2.395, interpMode=RCIM_Cubic, leaveTangent=-2.395, tangentMode=RCTM_User, time=0, value=0}
            {arriveTangent=0.8, interpMode=RCIM_Cubic, leaveTangent=0.8, tangentMode=RCTM_User, time=1, value=0.3081}
           ]}}}
        CurveIndex=&ReturnNormalizedExecIndex()
        Scale Curve=&Multiply_Float(
         A=&MakeFloatFromVector(Vector=$System.InterpolateOverTime.InterpolateOverTime_V3_.Moving Average, Channel=::NewEnumerator0)
         B=&Lerp_Float(
          A=0
          B=&Lerp_Float(A=25, B=100, Alpha=$System.Timeline.Time)
          Alpha=$System.InterpolateOverTime001.InterpolateOverTime_Float.Moving Average
         )
        )
       )
       Y=&FloatFromCurve(
        FloatCurve=DataInterface() {propertyValues {Curve {keys [{tangentMode=RCTM_User, time=0, value=0}, {tangentMode=RCTM_User, time=1, value=1}]}}}
        CurveIndex=&ReturnNormalizedExecIndex()
        Scale Curve=&FN_FrontEndFloatModifier(
         Gameplay=&Lerp_Float(A=50, B=200, Alpha=$System.Timeline.NormalizedTime)
         Front End=&Lerp_Float(A=0, B=200, Alpha=$System.Timeline002.NormalizedTime)
        )
       )
       Z=0
      )
     )
     Alpha=&ReturnNormalizedExecIndex()
    )) {enabled=false, isSetParameters=true, module=SetVariables_#5}
   &ScaleColor(
    Curve Index=$Particles.RibbonLinkOrder
    Linear Color Curve=DataInterface() {propertyValues {
      AlphaCurve {keys [<at 0.0886 1>,<at 0.9269 1>,<at 1 0>]}
      BlueCurve {keys [<at 0.9957 1>]}
      GreenCurve {keys [<at 0.9957 1>]}
      RedCurve {keys [<at 0.9957 1>]}
     }}
    Scale Mode=::NewEnumerator2
   )
   &ScaleColor(
    Scale Alpha=&FloatFromCurve(
     FloatCurve=DataInterface() {propertyValues {Curve {keys [{tangentMode=RCTM_User, time=0, value=1}, {tangentMode=RCTM_User, time=1, value=0}]}}}
     CurveIndex=$Particles.Initial.RibbonLinkOrder
     Scale Curve=&FN_FrontEndFloatModifier(
      Gameplay=&Lerp_Float(A=0.5, B=0.75, Alpha=$System.Timeline.NormalizedTime)
      Front End=&Lerp_Float(A=0, B=0.4, Alpha=$System.Timeline002.NormalizedTime)
     )
    )
    ScaleA=true
    ScaleRGB=true
   ) {module=ScaleColor001}
   /SetVariables(
    Particles.LineTrailColor:LinearColor=&MakeLinearColorFromVectorAndFloat(
     "Vector (RGB)"=<0.05 0.05 0.05>
     "Float (Alpha)"=&FloatFromCurve(
      FloatCurve=DataInterface() {propertyValues {Curve {keys [{tangentMode=RCTM_User, time=0, value=0}, <at 0.0841 1.0029>, {tangentMode=RCTM_User, time=0.7064, value=0}]}}}
      CurveIndex=$Particles.Initial.RibbonLinkOrder
      Scale Curve=&FN_FrontEndFloatModifier(
       Gameplay=&Multiply_Float(
        A=&Lerp_Float(A=0.3, B=0.45, Alpha=$System.Timeline.NormalizedTime)
        B=&FNTimeOfDayScalar(Night Scale=0.5, Morning Scale=0.75, Evening Scale=0.75)
       )
       Front End=&Lerp_Float(A=0.3, B=0.45, Alpha=$System.Timeline002.NormalizedTime)
      )
     )
    )
    Particles.LineTrailDynamic:Vector4f=&MakeVector4(
     X=&FloatFromCurve(FloatCurve=DataInterface() {propertyValues {Curve {keys [<at 0 1>,<at 1 0>]}}}, CurveIndex=$Particles.RibbonLinkOrder, Scale Curve=10)
     Y=0
     Z=0
     W=0
    )
    Particles.LineTrailWidth:NiagaraFloat=&FloatFromCurve(
     FloatCurve=DataInterface() {propertyValues {Curve {keys [
         {arriveTangent=1.4042, interpMode=RCIM_Cubic, leaveTangent=1.4042, tangentMode=RCTM_User, time=0, value=0.15}
         {tangentMode=RCTM_User, time=1, value=0.8}
        ]}}}
     CurveIndex=$Particles.Initial.RibbonLinkOrder
     Scale Curve=20
    )
   ) {isSetParameters=true}
   /SetVariables(
    Particles.SoftTrailColor:LinearColor=&Multiply_LinearColorByFloat(
     Linear Color=<rgba 1 1 1 1>
     Float=&FloatFromCurve(
      FloatCurve=DataInterface() {propertyValues {Curve {keys [<at 0 0.4206>,<at 0.0964 0.7912>,<at 0.7743 0>]}}}
      CurveIndex=$Particles.RibbonLinkOrder
      Scale Curve=&Multiply_Float(
       A=&FN_FrontEndFloatModifier(Gameplay=0.45, Front End=0.3)
       B=&FNTimeOfDayScalar(Night Scale=0.5, Morning Scale=0.75, Evening Scale=0.75)
      )
     )
    )
    Particles.SoftTrailDynamic:Vector4f=&MakeVector4(X=1, Y=1, Z=1, W=1)
    Particles.SoftTrailWidth=20
   ) {isSetParameters=true}
   &ScaleRibbonWidth(Ribbon Width Scale=&FloatFromCurve(
     FloatCurve=DataInterface() {propertyValues {Curve {keys [{tangentMode=RCTM_User, time=0, value=0.75}, <at 0.484 2.1885>, {tangentMode=RCTM_User, time=1, value=2}]}}}
     CurveIndex=$Particles.Initial.RibbonLinkOrder
    ))
   &DynamicMaterialParameters(
    Index 0 Param 1=&MaskFloat(Mask Source=0, Mask Parameter=?, Mask=0, Masked Float=0, Passthrough Float=1)
    Index 0 Param 2=1
    Index 0 Param 3=1
    Index 0 Param 4=1
    Module.Index0Float0Write=true
    Module.Index0Float1Write=true
    Module.Index0Float2Write=true
    Module.Index0Float3Write=true
   )
   &ScaleColor(
    Scale Alpha=&FNTimeOfDayScalar(Night Scale=0.25, Morning Scale=0.75, Evening Scale=0.75)
    ScaleA=true
    ScaleRGB=true
   ) {module=ScaleColor002}
  ]
  rendererClasses [&NiagaraSpriteRendererProperties, &NiagaraRibbonRendererProperties]
  renderers [
   {rendererClass=&NiagaraSpriteRendererProperties}
   {rendererClass=&NiagaraRibbonRendererProperties, rendererIndex=1}
   {rendererClass=&NiagaraRibbonRendererProperties, rendererIndex=2}
   {rendererClass=&NiagaraRibbonRendererProperties, rendererIndex=3}
  ]
 }
 MeggieTrail {
  bEnabled=false
  emitterUpdateScript [
   &EmitterState(Inactive Response=::NewEnumerator1, Life Cycle Mode=::NewEnumerator1, Loop Behavior=::NewEnumerator1, Loop Duration=0.9, UseLoopDelay=false)
   &SpawnBurst_Instantaneous(Spawn Time=0.45, Use Spawn Probability=false, UseLoopCountLimit=false)
  ]
  particleSpawnScript [&InitializeParticle(
    Color Mode=::NewEnumerator0
    Lifetime=0.4
    Sprite Rotation Angle=-90
    Sprite Rotation Mode=::NewEnumerator0
    Sprite Size=<20 20>
    Sprite Size Mode=::NewEnumerator4
    UsePositionOffset=false
    Write Material Random=true
   )]
  particleUpdateScript [
   &ParticleState()
   &SocketLocation(
    SamplePosOffset=true
    Sampled Position Offset=&SelectVectorFromArray(
     Direct Array Index=&MakeIntFromFloat(Float=&Add_Float(A=$User.RightDot, B=1))
     Vector Selection Array=DataInterface() {propertyValues {FloatData [<0 -4.5084 5>, <0 -4.5084 -0.4523>, <0 -4.5084 -5>]}}
    )
    Socket Index=5
    Socket Reader=$System.Socket Reader
    Socket Sampling Mode=::NewEnumerator3
    VelocitySpeedLimit=false
   )
   &ScaleSpriteSize(
    Scale Factor=&MakeVector2D(
     X=&FloatFromCurve(
      FloatCurve=DataInterface() {propertyValues {Curve {keys [<at 0.6496 1 RCIM_Cubic>, {arriveTangent=-7.8618, interpMode=RCIM_Cubic, leaveTangent=-7.8618, tangentMode=RCTM_User, time=0.9455, value=0}]}}}
      CurveIndex=$Emitter.NormalizedLoopAge
      Scale Curve=4
     )
     Y=1
    )
    Scale Sprite Size Mode=::NewEnumerator2
   )
   &ScaleColor(
    Scale Alpha=&FloatFromCurve(
     FloatCurve=DataInterface() {propertyValues {Curve {keys [<at 0.6496 1 RCIM_Cubic>, {arriveTangent=-7.8618, interpMode=RCIM_Cubic, leaveTangent=-7.8618, tangentMode=RCTM_User, time=0.9425, value=0}]}}}
     CurveIndex=$Emitter.NormalizedLoopAge
    )
    ScaleA=true
    ScaleRGB=true
   )
   &SpriteFacingAndAlignment(Sprite Alignment=<0 0 1>, WriteAlignment=true, WriteFacing=true)
   /SetVariables(Particles.SpriteSize=<100 20>) {enabled=false, isSetParameters=true}
  ]
  rendererClasses [&NiagaraSpriteRendererProperties]
  renderers [{rendererClass=&NiagaraSpriteRendererProperties}]
 }
 SpeedLine {
  emitterUpdateScript [
   &EmitterState()
   &SpawnRate(
    SpawnRate=&MakeCustomFloatFromBool(Bool=$User.bIsFrontEndPreview, True Float=50, False Float=80)
    Use Spawn Probability=true
   ) {module=SpawnRate001}
  ]
  particleSpawnScript [
   &InitializeParticle(
    Color=&MakeLinearColorFromVectorAndFloat(
     "Vector (RGB)"=<1 1 1>
     "Float (Alpha)"=&Multiply_Float(
      A=&UniformRangedFloat(Minimum=0.15, Maximum=0.3)
      B=&MakeCustomFloatFromBool(Bool=$User.bIsFrontEndPreview, True Float=0.5, False Float=1)
     )
    )
    Lifetime Max=&FN_FrontEndFloatModifier(Gameplay=0.35, Front End=0.2)
    Lifetime Min:NiagaraFloat=&FN_FrontEndFloatModifier(Gameplay=0.2, Front End=0.1)
    Lifetime Mode=::NewEnumerator1
    Sprite Size Max=<3 6>
    Sprite Size Min=<1 4>
    Sprite Size Mode=::NewEnumerator2
    UsePositionOffset=false
    Write Material Random=true
   )
   &TorusLocation(
    Align Torus With Axis=true
    Handle Radius=0
    Large Radius:NiagaraFloat=&RandomRangeFloat(Minimum=80, Maximum=150)
    MaskSpawnGroup=false
    Module.Override Randomness=false
    Offset=<0 -180 0>
    Torus Axis=<0 1 0>
    Torus Origin:NiagaraPosition=&Add_Vector(
     A=Linked() {linkedVariable {name=Particles.Position, type=NiagaraPosition}}
     B=&MakeVector(X=0, Y=&MakeCustomFloatFromBool(Bool=$User.bIsFrontEndPreview, True Float=0, False Float=0), Z=0)
    )
   )
   &AddVelocityFromPoint(
    Module.Use Falloff Distance=false
    Origin Offset=<0 -500 0>
    Velocity Strength=&Multiply_Float(A=600, B=&MakeCustomFloatFromBool(Bool=$User.bIsFrontEndPreview, True Float=1, False Float=1))
   )
   &AddVelocity(Velocity=&Multiply_Vector(
     A=&RandomRangeVector(Minimum=<0 300 0>, Maximum=<0 700 0>)
     B=&MakeVector(X=1, Y=&MakeCustomFloatFromBool(Bool=$User.bIsFrontEndPreview, True Float=1, False Float=1), Z=1)
    ))
  ]
  particleUpdateScript [
   &SolveForcesAndVelocity(Clamp Velocity=false, Limit Acceleration=false)
   &ParticleState()
   &ScaleSpriteSize(Scale Factor=&Vector2DFromCurve(
     Vector2Curve=DataInterface() {propertyValues {XCurve {keys [<at 0 0.5 RCIM_Cubic>,<at 0.5 1 RCIM_Cubic>]}, YCurve {keys [<at 0 0 RCIM_Cubic>,<at 0.5 1 RCIM_Cubic>]}}}
     ScaleCurve=&Vector2DFromFloat(Value=1)
    ))
   &ScaleSpriteSizeBySpeed(Max Scale Factor=<1 50>, Min Scale Factor=<0.01 0.05>, Velocity Threshold=1)
   &ScaleColor(
    Scale Alpha=&FloatFromCurve(
     FloatCurve=DataInterface() {propertyValues {Curve {keys [
         <at 0 0 RCIM_Cubic>
         <at 0.3759 1.0207>
         {arriveTangent=-2.221, interpMode=RCIM_Cubic, leaveTangent=-2.221, tangentMode=RCTM_User, time=1, value=0}
        ]}}}
     Scale Curve=&MakeCustomFloatFromBool(
      Bool=$User.bIsFrontEndPreview
      True Float=&MakeCustomFloatFromBool(Bool=$User.bIsFullyDeployed, True Float=1, False Float=0)
      False Float=&Lerp_Float(A=0, B=1, Alpha=$User.ForwardDot)
     )
    )
    Scale RGB=&VectorFromFloat(Value=&RandomRangeFloat(Minimum=0.5, Maximum=1))
    ScaleA=true
    ScaleRGB=true
   )
   &ScaleColor(
    Scale Alpha=&FNTimeOfDayScalar(Night Scale=0.5, Morning Scale=1, Evening Scale=1)
    ScaleA=true
    ScaleRGB=true
   ) {module=ScaleColor001}
  ]
  rendererClasses [&NiagaraSpriteRendererProperties]
  renderers [
   {rendererClass=&NiagaraSpriteRendererProperties}
   {rendererClass=&NiagaraSpriteRendererProperties, rendererIndex=1}
  ]
 }
}
summary {
 systemName=NS_FruitFire_Nudge_Glider
 emitters [
  {emitterName=Fan, rendererClasses [&NiagaraMeshRendererProperties]}
  {bEnabled=false, emitterName=MeggieTrail, rendererClasses [&NiagaraSpriteRendererProperties]}
  {emitterName=SpeedLine, rendererClasses [&NiagaraSpriteRendererProperties]}
  {emitterName=HairDryerLines, rendererClasses [&NiagaraSpriteRendererProperties]}
  {emitterName=FanLines, rendererClasses [&NiagaraSpriteRendererProperties]}
  {emitterName=FanTrail, rendererClasses [&NiagaraSpriteRendererProperties, &NiagaraRibbonRendererProperties]}
  {
   bEnabled=false
   emitterName=HairDryerTrail
   rendererClasses [&NiagaraSpriteRendererProperties, &NiagaraRibbonRendererProperties]
  }
  {
   bEnabled=false
   emitterName=HairDryerTrail001
   rendererClasses [&NiagaraSpriteRendererProperties, &NiagaraRibbonRendererProperties]
  }
  {
   emitterName=HairDryerTrail002
   rendererClasses [&NiagaraSpriteRendererProperties, &NiagaraRibbonRendererProperties]
  }
 ]
 userVariables [
  User.bIsFrontEnd {type=NiagaraBool}
  User.bIsFrontEndPreview {type=NiagaraBool}
  User.bIsFullyDeployed {type=NiagaraBool}
  User.ForwardDot {type=NiagaraFloat, defaultValue=0, description=""}
  User.RibbonIDOffset {type=NiagaraInt32}
  User.RightDot {type=NiagaraFloat, defaultValue=0, description=""}
 ]
}
