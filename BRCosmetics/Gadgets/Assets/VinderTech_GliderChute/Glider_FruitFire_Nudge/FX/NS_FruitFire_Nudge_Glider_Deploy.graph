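# Niagara system for deploy effects on the FruitFire Nudge glider chute, combining a propeller mesh with a sprite trail.
# Propellers bursts a delayed mesh that follows a glider socket, scales up with emitter age, drives dynamic material parameters and fades out late.
# MeggieTrail bursts a delayed socket-attached sprite that narrows and fades as its loop ends.
# User.RightDot selects one of three trail offsets by conversion to an array index.
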NiagaraSystem /BRCosmetics/Gadgets/Assets/VinderTech_GliderChute/Glider_FruitFire_Nudge/FX/NS_FruitFire_Nudge_Glider_Deploy
dependencies {
 usedDataInterfaces [
  &NiagaraDataInterfaceCurve
  &NiagaraDataInterfaceArrayFloat3
  &NiagaraDataInterfaceSocketReader
  &NiagaraDataInterfaceArrayFloat
 ]
 usedDynamicInputs [
  &SetBoolByFloatComparison
  &RemapRange
  &SelectFloatFromArray
  &FNTimeOfDayColor
 ]
 usedModules [
  &self:SystemSpawnScript.SystemScriptSource.SystemScriptGraph.NiagaraNodeAssignment_1.SetVariables_#1
  &SystemState
  &Timeline
  &self:SystemSpawnScript.SystemScriptSource.SystemScriptGraph.NiagaraNodeAssignment_0.SetVariables_#2
 ]
}
emitters {
 MeggieTrail {
  emitterUpdateScript [
   &EmitterState(
    Inactive Response=::NewEnumerator1
    Life Cycle Mode=::NewEnumerator1
    Loop Behavior=::NewEnumerator1
    Loop Delay=2.6
    Loop Duration=0.9
    UseLoopDelay=true
   )
   &SpawnBurst_Instantaneous(Spawn Time=0.45, Use Spawn Probability=false, UseLoopCountLimit=false)
  ]
  particleSpawnScript [&InitializeParticle(
    Color Mode=::NewEnumerator0
    Lifetime=0.4
    Sprite Rotation Angle=-90
    Sprite Rotation Mode=::NewEnumerator0
    Sprite Size=<20 20>
    Sprite Size Mode=::NewEnumerator4
    UsePositionOffset=false
    Write Material Random=true
   )]
  particleUpdateScript [
   &ParticleState()
   &SocketLocation(
    SamplePosOffset=true
    Sampled Position Offset=&SelectVectorFromArray(
     Direct Array Index=&MakeIntFromFloat(Float=&Add_Float(A=$User.RightDot, B=1))
     Vector Selection Array=DataInterface() {propertyValues {FloatData [<0 -4.5084 5>, <0 -4.5084 -0.4523>, <0 -4.5084 -5>]}}
    )
    Socket Index=5
    Socket Reader=$System.Socket Reader
    Socket Sampling Mode=::NewEnumerator3
    VelocitySpeedLimit=false
   )
   &ScaleSpriteSize(
    Scale Factor=&MakeVector2D(
     X=&FloatFromCurve(
      FloatCurve=DataInterface() {propertyValues {Curve {keys [<at 0.6496 1 RCIM_Cubic>, {arriveTangent=-7.8618, interpMode=RCIM_Cubic, leaveTangent=-7.8618, tangentMode=RCTM_User, time=0.9455, value=0}]}}}
      CurveIndex=$Emitter.NormalizedLoopAge
      Scale Curve=4
     )
     Y=1
    )
    Scale Sprite Size Mode=::NewEnumerator2
   )
   &ScaleColor(
    Scale Alpha=&FloatFromCurve(
     FloatCurve=DataInterface() {propertyValues {Curve {keys [<at 0.6496 1 RCIM_Cubic>, {arriveTangent=-7.8618, interpMode=RCIM_Cubic, leaveTangent=-7.8618, tangentMode=RCTM_User, time=0.9425, value=0}]}}}
     CurveIndex=$Emitter.NormalizedLoopAge
    )
    ScaleA=true
    ScaleRGB=true
   )
   &SpriteFacingAndAlignment(Sprite Alignment=<0 0 1>, WriteAlignment=true, WriteFacing=true)
   /SetVariables(Particles.SpriteSize=<100 20>) {enabled=false, isSetParameters=true}
  ]
  rendererClasses [&NiagaraSpriteRendererProperties]
  renderers [{rendererClass=&NiagaraSpriteRendererProperties}]
 }
 Propellers {
  emitterUpdateScript [
   &EmitterState(
    Inactive Response=::NewEnumerator1
    Life Cycle Mode=::NewEnumerator1
    Loop Behavior=::NewEnumerator1
    Loop Delay=1
    Loop Duration Mode=::NewEnumerator1
    UseLoopDelay=true
   )
   &SpawnBurst_Instantaneous(Use Spawn Probability=false, UseLoopCountLimit=false)
  ]
  particleSpawnScript [
   &InitializeParticle(
    Color=<rgba 0.5 0.5 0.5 1>
    Lifetime=3
    Mesh Scale Mode=::NewEnumerator3
    Mesh Uniform Scale=0.09
    Sprite Size Mode=::NewEnumerator3
    Uniform Sprite Size=80
    UsePositionOffset=false
    Write Material Random=true
   )
   &SocketLocation(
    Orientation Sampling=::NewEnumerator1
    SamplePosOffset=true
    Sampled Position Offset=<0 0 14>
    Socket Index=4
    Socket Reader=$System.Socket Reader
    Socket Sampling Mode=::NewEnumerator3
    VelocitySpeedLimit=false
   )
   &InitialMeshOrientation(Mesh Orientation Mode=::NewEnumerator2, Orientation Vector=<0 1 0>, Use Rotation Vector=false) {enabled=false}
  ]
  particleUpdateScript [
   &SolveForcesAndVelocity(Clamp Velocity=false, Limit Acceleration=false)
   &ParticleState()
   &SocketLocation(
    SamplePosOffset=true
    Sampled Position Offset=<0 0 14>
    Socket Index=4
    Socket Reader=$System.Socket Reader
    Socket Sampling Mode=::NewEnumerator3
    VelocitySpeedLimit=false
   ) {module=SocketLocation001}
   &DynamicMaterialParameters(
    Index 0 Param 1=0.2
    Index 0 Param 2=0.98
    Module.Param0WriteEnabled=true
    Module.Param1WriteEnabled=true
    Module.Param2WriteEnabled=false
    Module.Param3WriteEnabled=true
   )
   &ScaleColor(
    Scale Alpha=&FloatFromCurve(FloatCurve=DataInterface() {propertyValues {Curve {keys [<at 0.8645 1>,<at 0.8826 0>]}}})
    ScaleA=true
    ScaleRGB=true
   ) {module=ScaleColor002}
   &ScaleMeshSize(Scale Factor=&Multiply_VectorByFloat(
     Vector=<1 1 1>
     Float=&FloatFromCurve(FloatCurve=DataInterface() {propertyValues {Curve {keys [<at 0.6792 0>,<at 0.8785 1>]}}}, CurveIndex=$Emitter.Age)
    ))
  ]
  rendererClasses [&NiagaraMeshRendererProperties]
  renderers [{rendererClass=&NiagaraMeshRendererProperties, rendererIndex=0}]
 }
}
summary {
 systemName=NS_FruitFire_Nudge_Glider_Deploy
 emitters [
  {emitterName=Propellers, rendererClasses [&NiagaraMeshRendererProperties]}
  {emitterName=MeggieTrail, rendererClasses [&NiagaraSpriteRendererProperties]}
 ]
 userVariables [
  User.bIsFrontEnd {type=NiagaraBool}
  User.bIsFrontEndPreview {type=NiagaraBool}
  User.bIsFullyDeployed {type=NiagaraBool}
  User.RightDot {type=NiagaraFloat, defaultValue=0, description=""}
 ]
}
